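# Idle Niagara effect for the Tide King pickaxe that sheds short-lived, sprite-based drips from its skeletal mesh.
# A single emitter spawns by rate and movement distance around sampled mesh positions, with User.ColorVariant blending the spawn volume's shape and offset.
# Drips trail against owner velocity and fall under gravity while curves drive size and dynamic material parameters.
# A system platform switch scales movement-based spawn probability and the age threshold that gates collision.
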NiagaraSystem /BRCosmetics/Weapons/FORT_Melee/Pickaxe_TideKing/FX/NS_TideKing_Idle
dependencies {
 usedDataInterfaces [&NiagaraDataInterfaceCurve, &NiagaraDataInterfaceSkeletalMesh, &NiagaraDataInterfacePlatformSet]
 usedDynamicInputs [&FN_PickFloatDependOnPlatform]
 usedModules [
  &self:SystemSpawnScript.SystemScriptSource.SystemScriptGraph.NiagaraNodeAssignment_0.SetVariables_#1
  &self:SystemSpawnScript.SystemScriptSource.SystemScriptGraph.NiagaraNodeAssignment_1.SetVariables_#2
  &SystemState
 ]
}
emitters {Drips {
  emitterUpdateScript [
   &EmitterState()
   &SpawnRate(Spawn Probability=0.33, SpawnRate=2, Use Spawn Probability=true)
   &SpawnPerUnit(
    Module.Use Max Movement Threshold=true
    Module.Use Movement Tolerance=false
    Spawn Probability:NiagaraFloat=&Lerp_Float(A=0, B=0.5, Alpha=$System.PlatformSwitcher)
    Spawn Spacing=150
    Use Spawn Probability=true
   )
  ]
  particleSpawnScript [
   &SampleSkeletalMesh(
    AdditionalYawPitchRoll=false
    Mesh Sampling Type:ENiagaraSkelSamplingModeFull=::NewEnumerator5
    Skeletal Mesh:NiagaraDataInterfaceSkeletalMesh=$System.Skeletal Mesh
    Socket Sampling Mode:ENiagaraSocketSamplingMode=::NewEnumerator5
   )
   &InitializeParticle(
    Color Mode=::NewEnumerator0
    Lifetime Max=1
    Lifetime Min=0.7
    Lifetime Mode=::NewEnumerator1
    Sprite Size Mode=::NewEnumerator3
    Uniform Sprite Size=&UniformAOrBFloat(A=&RandomRangeFloat(Minimum=2, Maximum=5.5), B=6.5, Distribution Weight=0.1)
    UsePositionOffset=false
    Write Material Random=true
   )
   &ShapeLocation(
    MaskSpawnGroup=false
    Non Uniform Scale=&Lerp_Vector(A=<0.3 0.3 1.3>, B=<1 0.5 0.5>, Alpha=$User.ColorVariant)
    Offset=&Lerp_Vector(A=<0 0 -20>, B=<0 0 3>, Alpha=$User.ColorVariant)
    Offset Coordinate Space=::Local
    Offset Mode=::NewEnumerator0
    Override Seed=false
    Shape Origin=$Particles.SampleSkeletalMesh.SampledPosition
    Sphere Radius=18
    Sphere Surface Distribution=1
    UseOwnerScale=true
   )
   &AddVelocity(Module.Use Falloff Distance=false, Origin Offset=<0 0 25>, Override Seed=false, Velocity Mode=::NewEnumerator1)
   &AddVelocity(Velocity=&Multiply_VectorByFloat(
     Vector=&NormalizeVector(Vector To Normalize=$Engine.Owner.Velocity, Safe Normalize=true, Safe Fallback Threshold=0, Safe Fallback Vector=<0 0 0.001>)
     Float=-25
    )) {module=AddVelocity001}
   /Niagara/Modules/Update/SubUV/SubUVAnimation(SubUV Animation Mode=::NewEnumerator1) {module=SubUVAnimation001}
  ]
  particleUpdateScript [
   &ParticleState()
   &ScaleSpriteSize(Uniform Curve Sprite Scale=DataInterface() {propertyValues {Curve {keys [
        {interpMode=RCIM_Cubic, tangentMode=RCTM_Break, time=0, value=0}
        {interpMode=RCIM_Cubic, tangentMode=RCTM_Break, time=0.15, value=1}
        {interpMode=RCIM_Cubic, leaveTangent=-4.5487, tangentMode=RCTM_Break, time=0.5, value=1}
        <at 1 0 RCIM_Cubic>
       ]}}})
   &GravityForce(Gravity=&RandomRangeVector(Minimum=<0 0 -550>, Maximum=<0 0 -600>))
   &Collision(
    Age Colliding Particles=false
    Collision Enabled=&SetBoolByFloatComparison(A=$Particles.Age, B=&Lerp_Float(A=100, B=0.2, Alpha=$System.PlatformSwitcher))
    Friction=1
    Friction During a Bounce=1
    Module.Control Rest State Via Surface Normal=false
    Module.EnableMaxCollisionCount=false
    Module.Randomize Collision Normal Vector=false
    Restitution=0
   )
   &SolveForcesAndVelocity(Clamp Velocity=false, Limit Acceleration=false)
   &DynamicMaterialParameters(
    Index 0 Param 1=&RandomRangeFloat(Minimum=0, Maximum=1)
    Index 0 Param 2=&FloatFromCurve(FloatCurve=DataInterface() {propertyValues {Curve {keys [<at 0.35 0.45>,<at 1 0.9>]}}})
    Index 0 Param 3=&FloatFromCurve(FloatCurve=DataInterface() {propertyValues {Curve {keys [{arriveTangent=2.6651, interpMode=RCIM_Cubic, leaveTangent=2.6651, tangentMode=RCTM_User, time=0, value=0}, <at 1 1 RCIM_Cubic>]}}})
    Index 0 Param 4=&FloatFromCurve(FloatCurve=DataInterface() {propertyValues {Curve {keys [<at 0.5 0 RCIM_Cubic>, {arriveTangent=2.332, interpMode=RCIM_Cubic, leaveTangent=2.332, tangentMode=RCTM_User, time=0.75, value=1}]}}})
    Module.Param0WriteEnabled=true
    Module.Param1WriteEnabled=true
    Module.Param2WriteEnabled=false
    Module.Param3WriteEnabled=true
   ) {module=DynamicMaterialParameters002}
  ]
  rendererClasses [&NiagaraSpriteRendererProperties]
  renderers [{rendererClass=&NiagaraSpriteRendererProperties}]
 }}
summary {
 systemName=NS_TideKing_Idle
 emitters [{emitterName=Drips, rendererClasses [&NiagaraSpriteRendererProperties]}]
 userVariables [User.ColorVariant {type=NiagaraFloat, defaultValue=0, description=""}]
}
